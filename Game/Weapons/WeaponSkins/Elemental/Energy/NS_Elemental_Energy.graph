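# Niagara system for the Energy elemental weapon skin, layering energy meshes with electric arc and sparkle sprites.
# Elemental_Weapon_Setup and system variables drive mesh offset and size, sprite spawn-rate scale and cylinder-surface spawn shape.
# Both GPU emitters mix bursts on randomized loops with steady spawning, using spawn-group masks to create two particle variants each.
# Sprites copy spawn group into VisibilityTag for two renderers, and randomized dynamic material parameters vary the look.
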NiagaraSystem /Game/Weapons/WeaponSkins/Elemental/Energy/NS_Elemental_Energy
dependencies {
 usedDataInterfaces [&NiagaraDataInterfaceVectorCurve, &NiagaraDataInterfaceCurve, &NiagaraDataInterfaceSkeletalMesh]
 usedDynamicInputs []
 usedModules [&self:Elemental_Weapon_Setup, &SystemState]
}
emitters {
 Electric_Arcs_Sparkles_Sprite {
  simTarget=GPUComputeSim
  emitterUpdateScript [
   &EmitterState(
    Life Cycle Mode=::NewEnumerator1
    Loop Duration=&RandomRangeFloat(Minimum=0.1, Maximum=1.3)
    Recalculate Duration Each Loop=true
    UseLoopDelay=false
   )
   &SpawnBurst_Instantaneous(Spawn Group=1, Use Spawn Probability=false, UseLoopCountLimit=false) {module=SpawnBurst_Instantaneous001}
   &SpawnRate(SpawnGroup=1, SpawnRate=&Multiply_Float(A=25, B=$System.SpawnRateScale), Use Spawn Probability=false) {module=SpawnRate001}
   &SpawnRate(SpawnRate=&Multiply_Float(A=3, B=$System.SpawnRateScale), Use Spawn Probability=false)
   &SpawnBurst_Instantaneous(Use Spawn Probability=false, UseLoopCountLimit=false)
  ]
  particleSpawnScript [
   /SetVariables(Particles.VisibilityTag=&SpawnGroup()) {isSetParameters=true}
   &InitializeParticle(
    Color=&MaskLinearColorBySpawnGroup(Masked Linear Color=<rgba 20 14.0736 1.1 1>, Passthrough Linear Color=<rgba 20 20 20 1>)
    Lifetime=&MaskFloatBySpawnGroup(
     Masked Float=&RandomRangeFloat(Minimum=0.15, Maximum=0.3)
     Passthrough Float=&RandomRangeFloat(Minimum=0.15, Maximum=0.2)
    )
    Sprite Size=&MaskVector2dBySpawnGroup(
     Masked Vector2D=&RandomRangeVector2D(Minimum=<3 13>, Maximum=<5 21>)
     Passthrough Vector2D=&RandomRangeVector2D(Minimum=<0.4 0.4>, Maximum=<0.8 0.8>)
     Spawn Group Mask=0
    )
    Sprite Size Mode=::NewEnumerator4
    UsePositionOffset=false
    Write Material Random=true
   )
   &ShapeLocation(
    Cylinder Height=$System.CylinderHeight
    Cylinder Radius=$System.CylinderRadius
    MaskSpawnGroup=false
    Override Seed=false
    Rotation Mode=::NewEnumerator1
    Shape Origin=$System.CylinderStartLocation
    Shape Primitive=::NewEnumerator1
    Surface Only=true
    Use Endcaps In Surface Only Mode=false
    UseOwnerScale=true
   )
   &AddVelocity(Velocity=&Add_Vector(
     A=&Multiply_VectorByFloat(Vector=$Particles.ShapeLocation.ShapeNormal, Float=&RandomRangeFloat(Minimum=0, Maximum=0.02))
     B=&MaskVectorBySpawnGroup(Masked Vector=&RandomRangeVector(Minimum=<0.5 0 0>, Maximum=<1 0 0>), Passthrough Vector=<0 0 0>)
    ))
   &CameraOffset(Camera Offset Amount=20)
  ]
  particleUpdateScript [
   &ParticleState()
   &ScaleSpriteSize(
    Scale Sprite Size Mode=::NewEnumerator0
    Uniform Scale Factor=&MaskFloatBySpawnGroup(
     Masked Float=&FloatFromCurve(FloatCurve=DataInterface() {propertyValues {Curve {keys [{arriveTangent=1.5681, interpMode=RCIM_Cubic, leaveTangent=1.5681, tangentMode=RCTM_User, time=0, value=0.3}, <at 1 1 RCIM_Cubic>]}}})
     Passthrough Float=&FloatFromCurve(FloatCurve=DataInterface() {propertyValues {Curve {keys [
          <at 0 0 RCIM_Cubic>
          {arriveTangentWeight=0.1667, interpMode=RCIM_Cubic, leaveTangentWeight=0.25, tangentMode=RCTM_User, time=0.3, value=1}
          {arriveTangentWeight=0.2616, interpMode=RCIM_Cubic, leaveTangentWeight=0.1453, tangentMode=RCTM_User, time=0.75, value=1}
          <at 1 0 RCIM_Cubic>
         ]}}})
    )
   )
   &DynamicMaterialParameters(
    Index 0 Param 1=&RandomRangeFloat(Minimum=2.5, Maximum=3.5)
    Index 0 Param 2=&FloatFromCurve(FloatCurve=DataInterface() {propertyValues {Curve {keys [<at 0 1>,<at 1 -1>]}}})
    Index 0 Param 3=&RandomRangeFloat(Minimum=0.2, Maximum=0.7)
    Index 0 Param 4=&RandomRangeFloat(Minimum=-60, Maximum=60)
    Module.Index0Param0Write=true
    Module.Index0Param1Write=true
    Module.Index0Param2Write=true
    Module.Index0Param3Write=true
   )
   &SolveForcesAndVelocity(Clamp Velocity=false, Limit Acceleration=false)
  ]
  rendererClasses [&NiagaraSpriteRendererProperties]
  renderers [
   {rendererClass=&NiagaraSpriteRendererProperties}
   {rendererClass=&NiagaraSpriteRendererProperties, rendererIndex=1}
  ]
 }
 Energy_Mesh {
  simTarget=GPUComputeSim
  emitterUpdateScript [
   &EmitterState(
    Life Cycle Mode=::NewEnumerator1
    Loop Duration=&RandomRangeFloat(Minimum=0.15, Maximum=1.2)
    Recalculate Duration Each Loop=true
    UseLoopDelay=false
   )
   &SpawnBurst_Instantaneous(Spawn Group=1, Use Spawn Probability=false, UseLoopCountLimit=false)
   &SpawnRate(SpawnRate=8, Use Spawn Probability=false)
  ]
  particleSpawnScript [
   &InitializeParticle(
    Color=&MakeLinearColorFromVectorAndFloat(
     "Vector (RGB)"=<1 1 1>
     "Float (Alpha)"=&MaskFloatBySpawnGroup(
      Masked Float=&RandomRangeFloat(Minimum=0.85, Maximum=1)
      Passthrough Float=&RandomRangeFloat(Minimum=1.05, Maximum=1.1)
     )
    )
    Lifetime=&MaskFloatBySpawnGroup(
     Masked Float=&RandomRangeFloat(Minimum=1.6, Maximum=2.5)
     Passthrough Float=&RandomRangeFloat(Minimum=1.1, Maximum=1.6)
    )
    Mesh Scale=&Multiply_Vector(
     A=&MaskVectorBySpawnGroup(
      Masked Vector=&RandomRangeVector(Minimum=<1 0.3 0.3>, Maximum=<1 0.37 0.37>)
      Passthrough Vector=&RandomRangeVector(Minimum=<1 0.6 0.15>, Maximum=<1 0.8 0.25>)
     )
     B=$System.SizeScale_Energy
    )
    Mesh Scale Mode=::NewEnumerator4
    Position Offset=$System.Offset
    UsePositionOffset=true
    Write Material Random=true
   )
   &InitialMeshOrientation(
    Mesh Orientation Mode=::NewEnumerator1
    Rotation=&RandomRangeVector(Minimum=<0 0 0>, Maximum=<1 0 0>)
    Use Rotation Vector=true
   )
   &CameraOffset(Camera Offset Amount=20)
  ]
  particleUpdateScript [
   &ParticleState()
   &ScaleMeshSize(Scale Factor=&VectorFromCurve(VectorCurve=DataInterface() {propertyValues {
       XCurve {keys [<at 0 0.85>,<at 0.1335 0.8608>,<at 0.3609 0.9078>,<at 0.4996 0.9423>,<at 0.6209 0.971>,<at 0.8112 1>]}
       YCurve {keys [<at 0 1>,<at 1 1>]}
       ZCurve {keys [<at 0 0.7>,<at 0.1335 0.7888>,<at 0.3609 0.8809>,<at 0.4996 0.9283>,<at 0.6209 0.9639>,<at 0.8112 1>]}
      }}))
   &ScaleColor(
    Scale Alpha=&FloatFromCurve(FloatCurve=DataInterface() {propertyValues {Curve {keys [
         {arriveTangent=7.9524, interpMode=RCIM_Cubic, leaveTangent=7.9524, tangentMode=RCTM_User, time=0, value=0}
         <at 0.3 1 RCIM_Cubic>
         {arriveTangent=-1.7041, interpMode=RCIM_Cubic, leaveTangent=-1.7041, tangentMode=RCTM_User, time=1, value=0}
        ]}}})
    ScaleA=true
    ScaleRGB=true
   )
   &DynamicMaterialParameters(
    Index 0 Param 1=&MaskFloatBySpawnGroup(
     Masked Float=&RandomRangeFloat(Minimum=0.66, Maximum=0.78)
     Passthrough Float=&RandomRangeFloat(Minimum=1.6, Maximum=2.1)
    )
    Index 0 Param 2=&MaskFloatBySpawnGroup(
     Masked Float=&RandomRangeFloat(Minimum=0.25, Maximum=0.45)
     Passthrough Float=&RandomRangeFloat(Minimum=0.1, Maximum=0.15)
    )
    Index 0 Param 3=&RandomRangeFloat(Minimum=0.3, Maximum=0.6)
    Index 0 Param 4=&RandomRangeFloat(Minimum=-0.8, Maximum=0.8)
    Module.Index0Param0Write=true
    Module.Index0Param1Write=true
    Module.Index0Param2Write=true
    Module.Index0Param3Write=true
   )
   &SolveForcesAndVelocity(Clamp Velocity=false, Limit Acceleration=false)
  ]
  rendererClasses [&NiagaraMeshRendererProperties]
  renderers [{rendererClass=&NiagaraMeshRendererProperties, rendererIndex=0}]
 }
}
summary {
 systemName=NS_Elemental_Energy
 emitters [
  {emitterName=Energy_Mesh, simTarget=GPUComputeSim, rendererClasses [&NiagaraMeshRendererProperties]}
  {emitterName=Electric_Arcs_Sparkles_Sprite, simTarget=GPUComputeSim, rendererClasses [&NiagaraSpriteRendererProperties]}
 ]
 userVariables []
}
